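# Translucent, two-sided material for streak-line Niagara mesh particles in the main player's Congestion emote.
# A reflected gradient, repeated twice along U and sharpened by a power parameter, sets the erosion amount for an anisotropic noise texture.
# The inverted erosion result is multiplied separately by particle color and particle alpha, so the particle system controls tint and fade.
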Material /Game/Animation/Game/MainPlayer/Emotes/Congestion/FX/M_Emote_Congestion_StreakLines
graph {
 properties {
  bUsedWithNiagaraMeshParticles=true
  blendMode=BLEND_Translucent
  pixelDepthOffsetMode=PDOM_Legacy
  twoSided=true
  usedShadingModels=Default Lit
 }
 nodes [
  n1 = TextureSample(UVs=TextureCoordinate()) {
   automaticViewMipBias=false
   samplerType=SAMPLERTYPE_LinearGrayscale
   texture=&/T_Noise_Anisotropic
  }
  n2 = TextureCoordinate() {uTiling=2}
  n3 = MaterialFunctionCall!(Gradient=Frac(None=ComponentMask(None=@n2) {r=true}), Center Point=Constant() {r=0.5}) {materialFunction=&/MF_GradientReflected}
  n4 = ScalarParameter() {defaultValue=4, parameterName=GradientPower}
  n5 = OneMinus(None=MaterialFunctionCall!(Gradient=@n1.RGB, Erosion Amount=Power(Base=@n3.Result, Exp=@n4), Softness=ScalarParameter() {defaultValue=0.2}) {materialFunction=&/MF_Erosion}.Default Range)
  n6 = ParticleColor()
  Multiply(A=@n5, B=@n6.A)
  Multiply(A=@n5, B=@n6.RGB)
 ]
}
